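# Physics asset for the face accessory of the medium male Pirate 02 player body, covering its dangling cloth bones.
# Capsules on one horizontal and two chained vertical cloth bones are constrained back to the head, with narrow twist angles and wider swing angles.
# A head capsule and two neck spheres approximate the head and neck volume, with no constraints between them.
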PhysicsAsset /Game/Characters/Player/Male/Medium/Bodies/M_MED_Pirate02/Meshes/Parts/M_MED_Pirate_02_FaceAcc_Physics
bodies {
 dyn_cloth_horizontal [{shapeType=Capsule, length=2.5, radius=2, center=<4.2884 0 -0.6996>, rotation=<rot 20 0 0>}]
 dyn_cloth_vertical [{shapeType=Capsule, length=7.25, radius=1.25, center=<4.8393 -0.7505 -0.7121>, rotation=<rot 9.8641 -30.0538 15.3102>}]
 dyn_cloth_vertical_01 [{shapeType=Capsule, length=5.1805, radius=1.5842, center=<5.1483 0.0746 -0.457>, rotation=<rot 0 -90.8791 -47.4282>}]
 head [{shapeType=Capsule, length=9.25, radius=6.25, rotation=<rot 90 0.0099 0.0099>}]
 neck_01 [{shapeType=Sphere, radius=7}]
 neck_02 [{shapeType=Sphere, radius=6.5}]
}
constraints [
 {bone1Name=dyn_cloth_horizontal, bone2Name=head, swing1LimitDegrees=45, swing2LimitDegrees=30, twistLimitDegrees=5}
 {bone1Name=dyn_cloth_vertical, bone2Name=head, swing1LimitDegrees=60, swing2LimitDegrees=36, twistLimitDegrees=5}
 {bone1Name=dyn_cloth_vertical_01, bone2Name=dyn_cloth_vertical, swing1LimitDegrees=30, swing2LimitDegrees=30, twistLimitDegrees=5}
]
